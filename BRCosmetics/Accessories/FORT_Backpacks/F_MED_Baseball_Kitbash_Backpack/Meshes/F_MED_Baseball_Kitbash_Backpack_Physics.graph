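# Physics asset for the Baseball Kitbash backpack cosmetic, defining collision and joint limits for its bat and handle.
# Single capsule bodies cover the backpack_root, dyn_bat and dyn_handle bones.
# The bat and handle each constrain directly to the root: the bat has tight swing and twist limits, while the handle swings widely on one axis and stays tight on the others.
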PhysicsAsset /BRCosmetics/Accessories/FORT_Backpacks/F_MED_Baseball_Kitbash_Backpack/Meshes/F_MED_Baseball_Kitbash_Backpack_Physics
bodies {
 backpack_root [{shapeType=Capsule, length=19.3858, radius=5.8179, center=<-6.6894 -5.7635 0.3264>, rotation=<rot 0 125.3999 -88.1824>}]
 dyn_bat [{shapeType=Capsule, length=23.077, radius=3.1326, center=<-1.2 0.3796 0.0762>, rotation=<rot 0 93.3123 -85.5224>}]
 dyn_handle [{shapeType=Capsule, length=14.1599, radius=6.4962, center=<-4.7921 0.7572 -0.0001>, rotation=<rot 90 -3.5763 1.4214>}]
}
constraints [
 {bone1Name=dyn_bat, bone2Name=backpack_root, swing1LimitDegrees=10, swing2LimitDegrees=10, twistLimitDegrees=3}
 {bone1Name=dyn_handle, bone2Name=backpack_root, swing1LimitDegrees=66.4286, swing2LimitDegrees=5, twistLimitDegrees=5}
]
